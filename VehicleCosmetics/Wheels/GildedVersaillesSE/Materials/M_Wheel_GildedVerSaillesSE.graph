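# Translucent, two-sided glow material for the Gilded Versailles SE cosmetic wheel, with usage flags for skeletal meshes and mesh deformers.
# Green and red channels of the FX texture, sampled through two identically configured time-driven panners with adjustable tiling, are summed.
# A gradient texture masks the sum and adds a faint base, driving opacity scaled by a parameter.
# Emissive is a uniform Flare Color times an intensity, so the animated pattern lives only in opacity.
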Material /VehicleCosmetics/Wheels/GildedVersaillesSE/Materials/M_Wheel_GildedVerSaillesSE
graph {
 properties {
  bAutomaticallySetUsageInEditor=false
  bUseMaterialAttributes=true
  bUsedWithMeshDeformer=true
  bUsedWithSkeletalMesh=true
  blendMode=BLEND_Translucent
  twoSided=true
  usedShadingModels=Default Lit
 }
 nodes [
  n1 = TextureCoordinate()
  n2 = AppendVector(
   A=Multiply(A=ComponentMask(None=@n1) {r=true}, B=ScalarParameter() {defaultValue=1, parameterName=Glow_TileX})
   B=Multiply(A=ComponentMask(None=@n1) {g=true}, B=ScalarParameter() {defaultValue=0.1, parameterName=Glow_TileY})
  )
  n3 = Multiply(A=Time(), B=ScalarParameter() {defaultValue=1, parameterName=Speed})
  n4 = TextureSample(UVs=Panner(Coordinate=@n2, Time=@n3, Speed=Constant2Vector() {g=0.1, r=0.1}.G) {speedX=0.1, speedY=0.1}) {texture=/VehicleCosmetics/Wheels/GildedVersaillesSE/Textures/T_Wheel_GildedVerSaillesSE_Fx}
  n5 = TextureSample(UVs=Panner(Coordinate=@n2, Time=@n3, Speed=Constant2Vector() {g=0.1, r=0.1}.G) {speedX=0.1, speedY=0.1}) {texture=&T_Wheel_GildedVerSaillesSE_Fx}
  n6 = TextureSampleParameter2D(UVs=TextureCoordinate()) {parameterName=Param, texture=/VehicleCosmetics/Wheels/GildedVersaillesSE/Textures/T_Wheel_GildedVerSaillesSE_Fx_Gradient}
  n7 = Add(A=Multiply(A=Add(A=@n4.G, B=@n5.R), B=@n6.R), B=Multiply(A=@n6.R, B=Constant() {r=0.025}))
  SetMaterialAttributes!(
   MaterialAttributes
   Emissive Color=Multiply(
    A=VectorParameter() {parameterName=Flare Color, defaultValue=<rgba 1 0.5474 0.2039 1>}.RGB
    B=ScalarParameter() {defaultValue=10, parameterName=Emissive}
   )
   Opacity=Multiply(A=Saturate(None=@n7), B=ScalarParameter() {defaultValue=1, parameterName=Opacity, sliderMax=1})
  ) {attributeSetTypes [#1, #2]}
 ]
}
