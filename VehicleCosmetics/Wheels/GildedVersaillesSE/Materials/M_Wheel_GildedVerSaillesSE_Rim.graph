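# Two-sided, lit rim material for the Gilded Versailles SE vehicle cosmetic wheel, built on MF_WheelParent_Simple.
# Vertex color red blends rim and tire normal maps, and selects rim mask red or tire mask green.
# That mask blends parent emissive toward base color times a normal-mapped Fresnel, whose exponent and base reflect fraction both read Fresnel_a.
# The result passes through vehicle damage and de-rez transition functions.
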Material /VehicleCosmetics/Wheels/GildedVersaillesSE/Materials/M_Wheel_GildedVerSaillesSE_Rim
graph {
 properties {
  bAutomaticallySetUsageInEditor=false
  bUseMaterialAttributes=true
  bUsedWithMeshDeformer=true
  bUsedWithSkeletalMesh=true
  twoSided=true
  usedShadingModels=Default Lit
 }
 nodes [
  n1 = MaterialFunctionCall() {materialFunction=&MF_WheelParent_Simple}
  n2 = GetMaterialAttributes(None=@n1.MaterialAttributes) {attributeGetTypes [#1, #2]}
  n3 = VectorParameter() {group=0_Adjust Fresnel, channelNames {a="", b="", g=Fresnel_b, r=Fresnel_a}, defaultValue=<rgba 5 0.04 0 0>}
  n4 = TextureSampleParameter2D() {
   parameterName=Rim_Normal
   samplerType=SAMPLERTYPE_Normal
   texture=/VehicleCosmetics/Wheels/GildedVersaillesSE/Textures/T_Rim_GildedVersaillesSE_N
  }
  n5 = TextureSampleParameter2D() {
   parameterName=Tire_Normal
   samplerType=SAMPLERTYPE_Normal
   texture=/VehicleCosmetics/Wheels/GildedVersaillesSE/Textures/T_Tire_GildedVersaillesSE_N
  }
  n6 = Transform(None=LinearInterpolate(A=@n4.RGB, B=@n5.RGB, Alpha=@n1.VertexColorR))
  n7 = Multiply(
   A=@n2.Base Color
   B=Multiply(
    A=Fresnel(ExponentIn=@n3.Fresnel_a, BaseReflectFractionIn=@n3.Fresnel_a, Normal=@n6)
    B=ScalarParameter() {defaultValue=1, parameterName=Fresnel}
   )
  )
  n8 = LinearInterpolate(A=@n7, B=@n7, Alpha=@n1.VertexColorR)
  n9 = ComponentMask(None=@n1.RimRGBAMask) {r=true}
  n10 = LinearInterpolate(A=@n8, B=@n8, Alpha=@n9)
  Add(A=@n2.Emissive Color, B=@n10)
  n11 = LinearInterpolate(A=@n2.Emissive Color, B=@n10, Alpha=LinearInterpolate(A=@n9, B=ComponentMask(None=@n1.TireRGBAMask) {g=true}, Alpha=@n1.VertexColorR))
  n12 = LinearInterpolate(A=@n4.RGB, B=@n5.RGB, Alpha=@n1.VertexColorR)
  MaterialFunctionCall(In=MaterialFunctionCall(Material=SetMaterialAttributes!(MaterialAttributes=@n2.MaterialAttributes, Emissive Color=@n11, Normal=@n12) {attributeSetTypes [#2, #3]}) {materialFunction=&MF_BaseVehicle_Damage}.Result) {materialFunction=&MF_VehicleDeRez_Transition}
 ]
}
